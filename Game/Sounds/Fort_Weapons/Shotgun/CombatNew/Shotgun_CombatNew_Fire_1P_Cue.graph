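# Sound cue for the first-person firing sound of the CombatNew shotgun.
# A single close-fire wave feeds a Modulator that pins volume and leaves pitch modulation at its default, so there is no sample randomization.
# The cue lowers overall volume and references shared first-person weapon attenuation, a ranged-fire 1P sound class and a first-person gunfire concurrency set to govern falloff, mixing and voice limits.
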SoundCue /Game/Sounds/Fort_Weapons/Shotgun/CombatNew/Shotgun_CombatNew_Fire_1P_Cue
root=Modulator_7
nodes [
 {id=Modulator_7, children [WavePlayer_3], properties {volumeMax=1, volumeMin=1}}
 {id=WavePlayer_3, properties {soundWaveAssetPtr=&/Shotgun_CombatNew_Fire_Close_01}}
]
properties {
 attenuationSettings=&Fort_Weapon_Projectile_FirstPerson_Athena_Att
 maxDistance=2097152
 soundClassObject=&Fort_WeaponRanged_Fire_1P_PSM
 volumeMultiplier=0.7
 attenuationOverrides {falloffDistance=5000, attenuationShapeExtents {x=1000}}
 concurrencyOverrides {bLimitToOwner=true, maxCount=3}
 concurrencySet [&Weapon_Gunfire_FirstPerson_Conc]
}
