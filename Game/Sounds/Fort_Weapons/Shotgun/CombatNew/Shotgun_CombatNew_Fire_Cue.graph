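# Sound cue for the CombatNew shotgun's firing sound, built from close, medium and distant layers.
# A mixer plays all three waves together, each through its own volume modulator, distance attenuation and sound class override, with close weighted highest.
# Medium and distant share the distant-fire class, and only the distant layer's modulator varies pitch downward.
# Local and global gunfire concurrency sets are assigned to limit overlapping shots.
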SoundCue /Game/Sounds/Fort_Weapons/Shotgun/CombatNew/Shotgun_CombatNew_Fire_Cue
root=Mixer_3
nodes [
 {
  id=Attenuation_1
  children [Modulator_1]
  properties {attenuationSettings=&Fort_Weapon_Projectile_Medium_Mono_Athena_Att}
 }
 {id=Attenuation_2, children [Modulator_7], properties {attenuationSettings=&Fort_Weapon_Projectile_Close_Athena_Att}}
 {
  id=Attenuation_3
  children [Modulator_6]
  properties {attenuationSettings=&Fort_Weapon_Projectile_Distant_Mono_Athena_Att}
 }
 {id=Mixer_3, children [SoundClass_2,SoundClass_6,SoundClass_3], properties {inputVolume [1,0.8,0.8]}}
 {id=Modulator_1, children [WavePlayer_20], properties {volumeMax=2, volumeMin=2}}
 {id=Modulator_6, children [WavePlayer_21], properties {pitchMax=1, pitchMin=0.9, volumeMax=2, volumeMin=2}}
 {id=Modulator_7, children [WavePlayer_19], properties {volumeMax=1.3, volumeMin=1.25}}
 {
  id=SoundClass_2
  children [Attenuation_2]
  properties {soundClassOverride=&Fort_WeaponRanged_Fire_3P_PSM}
 }
 {
  id=SoundClass_3
  children [Attenuation_3]
  properties {soundClassOverride=&Fort_WeaponRanged_Fire_Distant}
 }
 {
  id=SoundClass_6
  children [Attenuation_1]
  properties {soundClassOverride=&Fort_WeaponRanged_Fire_Distant}
 }
 {id=WavePlayer_19, properties {soundWaveAssetPtr=&/Shotgun_CombatNew_Fire_Close_01}}
 {id=WavePlayer_20, properties {soundWaveAssetPtr=&/Shotgun_CombatNew_Fire_Medium_01}}
 {id=WavePlayer_21, properties {soundWaveAssetPtr=&/Shotgun_CombatNew_Fire_Distant_01}}
]
properties {
 maxDistance=27500
 volumeMultiplier=0.5
 attenuationOverrides {falloffDistance=5000, attenuationShapeExtents {x=1000}}
 concurrencyOverrides {bLimitToOwner=true, maxCount=3}
 concurrencySet [&Weapon_Gunfire_Local_Conc, &Weapon_Gunfire_Global_Conc]
}
