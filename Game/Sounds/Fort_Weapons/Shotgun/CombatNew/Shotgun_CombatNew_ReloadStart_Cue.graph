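# Sound cue for the start of the CombatNew shotgun reload in Fort weapon audio.
# A Modulator feeds from a single wave player holding the Reload_Start_01 wave.
# It plays through the weapon foley sound class with the shared reload attenuation and a reduced volume multiplier.
# Concurrency is overridden with a rule that prevents new instances once the limit is reached.
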SoundCue /Game/Sounds/Fort_Weapons/Shotgun/CombatNew/Shotgun_CombatNew_ReloadStart_Cue
root=Modulator_0
nodes [
 {id=Modulator_0, children [WavePlayer_1]}
 {id=WavePlayer_1, properties {soundWaveAssetPtr=&/Shotgun_CombatNew_Reload_Start_01}}
]
properties {
 attenuationSettings=&Fort_Weapon_Reload_Att
 bOverrideConcurrency=true
 maxDistance=2250
 soundClassObject=&Fort_WeaponFoley
 volumeMultiplier=0.601
 attenuationOverrides {falloffDistance=1600}
 concurrencyOverrides {resolutionRule=PreventNew}
}
